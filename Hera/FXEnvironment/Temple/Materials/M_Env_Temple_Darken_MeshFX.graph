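# Unlit Modulate-blend material for Temple environment mesh FX that darkens or tints whatever is behind the mesh.
# A cyan-tinted color parameter is multiplied by a scalar picked through a time-of-day function with morning, day, evening and night values.
# A view-angle fade blends from white, which is neutral under Modulate, to that color, so the effect varies with viewing angle.
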Material /Hera/FXEnvironment/Temple/Materials/M_Env_Temple_Darken_MeshFX
graph {
 properties {
  bAutomaticallySetUsageInEditor=false
  blendMode=BLEND_Modulate
  previewMesh=/Game/Environments/Hera/FX/Meshes/SM_WaterTankBreak_GlassCracks_Water_01
  shadingModel=MSM_Unlit
  usedShadingModels=Unlit
 }
 nodes [
  n1 = ScalarParameter() {defaultValue=0.9, parameterName=Value_Morning}
  n2 = ScalarParameter() {defaultValue=0.7, parameterName=Value_Day}
  n3 = ScalarParameter() {defaultValue=0.7, parameterName=Value_Evening}
  n4 = ScalarParameter() {defaultValue=1.5, parameterName=Value_Night}
  LinearInterpolate(
   B=Multiply(
    A=VectorParameter() {parameterName="Color (Modulate)", defaultValue=<rgba 0.1953 0.9241 1 1>}.RGB
    B=MaterialFunctionCall(Value_Morning=@n1, Value_Day=@n2, Value_Evening=@n3, Value_Night=@n4) {materialFunction=&Scalar_Time_of_Day}.Result
   )
   Alpha=MaterialFunctionCall("FadeSensitivity (optional)"=ScalarParameter() {defaultValue=4, parameterName=Angle Fade}) {materialFunction=&FadeBasedOnViewAngle}.Result
  ) {constA=1}
 ]
}
