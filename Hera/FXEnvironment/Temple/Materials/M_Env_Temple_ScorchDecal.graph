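# Translucent unlit deferred decal material for Temple environment FX, drawing a scorch strip with glowing embers.
# Panning noise warps a tiled ember texture, whose red channel feeds all three inputs of a time-driven Lerp_3Color.
# That result is power-sharpened, sphere-masked, tinted orange by default and passed through MF_FX_EmissiveAdjustments.
# Opacity multiplies the scorch mask's green channel by a falloff along U, decal lifetime opacity and a scalar parameter.
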Material /Hera/FXEnvironment/Temple/Materials/M_Env_Temple_ScorchDecal
graph {
 properties {
  bAutomaticallySetUsageInEditor=false
  blendMode=BLEND_Translucent
  materialDomain=MD_DeferredDecal
  previewMesh=/Game/Environments/Hera/FX/Meshes/SM_WaterTankBreak_GlassCracks_Water_01
  shadingModel=MSM_Unlit
  usedShadingModels=Unlit
 }
 nodes [
  Constant()
  Constant() {r=1}
  n1 = TextureSampleParameter2D(UVs=Panner(Coordinate=TextureCoordinate()) {speedX=0.1}) {parameterName=Emissive Warp, samplerType=SAMPLERTYPE_LinearColor, texture=/Game/Effects/Fort_Effects/Textures/Tile/T_Random_Dots}
  n2 = ScalarParameter() {defaultValue=0.05, parameterName=Warp Amount}
  n3 = TextureCoordinate()
  n4 = AppendVector(A=ScalarParameter() {defaultValue=2, parameterName=Emissive Tile X}, B=ScalarParameter() {defaultValue=1, parameterName=Emissive Tile Y})
  n5 = TextureSampleParameter2D(UVs=Add(A=Multiply(A=@n1.G, B=@n2) {constB=0.05}, B=Multiply(A=@n3, B=@n4))) {parameterName="Emissive Embers (Packed)", texture=/Game/Effects/Fort_Effects/Textures/Tile/T_Dots_Small}
  n6 = Frac(None=Multiply(A=Time() {period=3}) {constB=0.5})
  n7 = Abs(None=Subtract(A=Multiply(A=@n6) {constB=2}))
  n8 = Saturate(None=Power(
    Base=MaterialFunctionCall(A=@n5.R, B=@n5.R, C=@n5.R, Alpha=@n7) {materialFunction=&Lerp_3Color}.Result
    Exp=ScalarParameter() {defaultValue=2, parameterName=EmissiveMaskPower}
   ))
  n9 = TextureCoordinate()
  n10 = Constant() {r=0.5}
  n11 = Saturate(None=Power(Base=SphereMask(A=@n9, B=@n10) {attenuationRadius=0.5, hardnessPercent=0}))
  MaterialFunctionCall(Emissive=Multiply(A=Multiply(A=@n8, B=@n11), B=VectorParameter() {parameterName=Color, defaultValue=<rgba 5 0.5977 0.1026 1>}.RGB)) {materialFunction=&MF_FX_EmissiveAdjustments}
  n12 = TextureSampleParameter2D() {parameterName=Decal Alpha, samplerType=SAMPLERTYPE_Masks, texture=/Hera/FXEnvironment/Temple/Textures/T_FX_Decal_ScorchStrip}
  n13 = TextureCoordinate()
  n14 = Power(Base=OneMinus(None=ComponentMask(None=@n13) {r=true})) {constExponent=0.75}
  Multiply(
   A=ScalarParameter() {defaultValue=1.5, parameterName=Opacity Multiply}
   B=Multiply(A=Multiply(A=@n12.G, B=@n14), B=DecalLifetimeOpacity().Opacity)
  )
 ]
}
