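# Unlit translucent sparkle material for the Temple environment, drawing cross-flare sprites tinted by time of day.
# Vertex color red varies size and brightness, while green varies pulse speed and upward drift applied through world position offset.
# Sprites pulse, shrink near the camera, offset along the view, and adjust scale and opacity with distance.
# Weather-event intensity from the Fortnite parameter collection scales sprites away, and low quality skips depth fade.
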Material /Hera/FXEnvironment/Temple/Materials/M_Env_Temple_Sparkles
graph {
 properties {
  bUseTranslucencyVertexFog=false
  blendMode=BLEND_Translucent
  opacityMaskClipValue=0.01
  shadingModel=MSM_Unlit
  twoSided=true
  usedShadingModels=Unlit
 }
 nodes [
  n1 = Frac(None=Multiply(A=NamedRerouteUsage()) {constB=2})
  n2 = LinearInterpolate(
   A=ScalarParameter() {parameterName=WPO Upwards Min, sliderMax=50}
   B=ScalarParameter() {parameterName=WPO Upwards Max, sliderMax=50}
   Alpha=Saturate(None=VertexColor().G)
  )
  n3 = LinearInterpolate(
   A=ScalarParameter() {defaultValue=25, group=WPO, parameterName=SpriteSizeMin}
   B=ScalarParameter() {defaultValue=85, group=WPO, parameterName=SpriteSizeMax}
   Alpha=Saturate(None=VertexColor().R)
  )
  n4 = LinearInterpolate(A=Multiply(A=@n3, B=ScalarParameter() {parameterName=Pulsing Shrink}), B=@n3, Alpha=NamedRerouteUsage())
  n5 = MaterialFunctionCall(Distance=ScalarParameter() {defaultValue=500, group=WPO, parameterName=NearCamShrinkDist}) {materialFunction=&Near_Camera_Fade}
  n6 = Add(
   A=MaterialFunctionCall!(Center Location=WorldPosition().XYZ, 0-1 UVs, XY scale=Multiply(A=@n4, B=@n5.Result), OverridePivot, "Normals (Optional)") {materialFunction=&Sprite}.World position offset
   B=MaterialFunctionCall(Offset Amount=ScalarParameter() {defaultValue=25, group=WPO, parameterName=Camera Offset Amount}) {materialFunction=/Engine/Functions/Engine_MaterialFunctions02/WorldPositionOffset/CameraOffset}.Clamped Camera Offset
  )
  n7 = Power(Base=MaterialFunctionCall(
    Fade_Start_Distance=ScalarParameter() {defaultValue=9000, parameterName=EndScale_StartDistance}
    Fade_End_Distance=ScalarParameter() {defaultValue=800, parameterName=EndScale_EndDistance}
   ) {materialFunction=&MF_Distance_SyncedFade_FX}.Result) {constExponent=0.85}
  n8 = LinearInterpolate(
   A=ScalarParameter() {defaultValue=1, parameterName=Distance Scale Low}
   B=ScalarParameter() {defaultValue=4.5, parameterName=Distance Scale High}
   Alpha=NamedRerouteDeclaration(None=Saturate(None=@n7)) {name=DistanceScale, nodeColor=<rgba 0 1 0.1094 1>}
  ) {constA=1}
  n9 = ScalarParameter() {defaultValue=1, parameterName=OutsideStormScale}
  n10 = ScalarParameter() {parameterName=InsideStormScale}
  n11 = CollectionParameter() {
   collection=&FortniteMaterialParameters
   parameterName=CurrentWeatherEventIntensity
  }
  n12 = Clamp(None=@n11)
  Add(
   A=Multiply(A=Multiply(A=@n1, B=@n2), B=Constant3Vector() {constant=<rgba 0 0 1 0>})
   B=Multiply(A=Multiply(A=@n6, B=@n8), B=LinearInterpolate(A=@n9, B=@n10, Alpha=@n12))
  )
  n13 = Time()
  n14 = VertexColor()
  n15 = LinearInterpolate(
   A=ScalarParameter() {defaultValue=0.25, parameterName=Time Speed Min}
   B=ScalarParameter() {defaultValue=0.6, parameterName=Time Speed Max}
   Alpha=Saturate(None=@n14.G)
  )
  n16 = Sine(None=NamedRerouteDeclaration(None=Multiply(A=@n13, B=@n15)) {name=Scaled Time, nodeColor=<rgba 0 1 0.6528 1>})
  ConstantBiasScale(None=@n16)
  Constant() {r=1}
  n17 = VectorParameter() {parameterName=Color_Morning, defaultValue=<rgba 0.1178 0.4206 1 1>}
  n18 = VectorParameter() {parameterName=Color_Day, defaultValue=<rgba 0.3207 0.9731 2 1>}
  n19 = VectorParameter() {parameterName=Color_Evening, defaultValue=<rgba 0.1072 0.3059 0.6562 1>}
  n20 = VectorParameter() {parameterName=Color_Night, defaultValue=<rgba 0 0.1106 0.4635 1>}
  MaterialFunctionCall(Emissive=Multiply(
    A=MaterialFunctionCall(Color_Morning=@n17.RGB, Color_Day=@n18.RGB, Color_Evening=@n19.RGB, Color_Night=@n20.RGB) {materialFunction=&Color_Time_of_Day}.Result
    B=LinearInterpolate(Alpha=@n14.R) {constA=0.8, constB=1.5}
   )) {materialFunction=&MF_FX_EmissiveAdjustments}
  n21 = NamedRerouteDeclaration(None=@n16) {name=Time Pulse, nodeColor=<rgba 1 0 0.5625 1>}
  n22 = ScalarParameter() {defaultValue=2, parameterName=AlphaStrengthMult}
  n23 = TextureSampleParameter2D() {parameterName=FlareTexture, texture=/Game/Effects/Fort_Effects/Textures/Flares/T_Cross_Flare_ColorsRGB}
  n24 = Multiply(
   A=Clamp(None=Multiply(A=@n21, B=@n22) {constB=50}, Min=ScalarParameter() {parameterName=Alpha Min}, Max=@n22) {maxDefault=50, minDefault=0.25}
   B=StaticComponentMaskParameter(None=AppendVector(A=@n23.RGB, B=@n23.RGB)) {defaultG=true, parameterName=FlareTexture_Mask}
  )
  n25 = ScalarParameter() {defaultValue=50, group=Opacity, parameterName=Opacity_Depth_FadeDistance}
  Multiply(
   A=LinearInterpolate(
    A=ScalarParameter() {defaultValue=1, parameterName=Distance Opacity Close}
    B=ScalarParameter() {defaultValue=0.5, parameterName=Distance Opacity Far}
    Alpha=NamedRerouteUsage()
   ) {constA=1}
   B=QualitySwitch(Default=DepthFade(Opacity=@n24, FadeDistance=@n25), Low=@n24)
  )
  NamedRerouteUsage()
 ]
}
