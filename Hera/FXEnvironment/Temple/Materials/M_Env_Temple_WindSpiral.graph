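# Translucent unlit, two-sided material for temple wind-spiral effects on Niagara mesh particles and ribbons.
# Panning plasma-coil and smoke-trail textures, distorted by scrolling smoke noise, form a tinted emissive scaled by time of day.
# Opacity combines those trails, U-axis end fades, particle alpha and view-angle fading, with camera and depth fades skipped on forward/mobile.
# Vertex offset spins the mesh around Z with a sine wobble, correcting normals for the angle fade.
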Material /Hera/FXEnvironment/Temple/Materials/M_Env_Temple_WindSpiral
graph {
 properties {
  bUsedWithNiagaraMeshParticles=true
  bUsedWithNiagaraRibbons=true
  blendMode=BLEND_Translucent
  pixelDepthOffsetMode=PDOM_Legacy
  shadingModel=MSM_Unlit
  twoSided=true
  usedShadingModels=Unlit
 }
 nodes [
  n1 = ScalarParameter() {defaultValue=1, parameterName=MultTOD_Morning}
  n2 = ScalarParameter() {defaultValue=1, parameterName=MultTOD_Day}
  n3 = ScalarParameter() {defaultValue=1, parameterName=MultTOD_Evening}
  n4 = ScalarParameter() {defaultValue=1, parameterName=MultTOD_Night}
  n5 = TextureCoordinate()
  n6 = Multiply(A=TextureSample(UVs=Panner(Coordinate=@n5) {speedX=0.1}) {texture=&T_Inky_Smoke_Tile}.R) {constB=0.1}
  n7 = Multiply(A=Time(), B=ScalarParameter() {defaultValue=0.4, parameterName=Speed Direction}) {constB=-1.5}
  n8 = Panner(
   Coordinate=Add(A=@n6, B=TextureCoordinate() {uTiling=0.5, vTiling=0.5})
   Time=@n7
   Speed=VectorParameter() {parameterName=ColorNoiseSpeed, defaultValue=<rgba -0.8 0 0 0>}.RGB
  ) {speedX=-0.8}
  n9 = Multiply(
   A=Desaturation(
    None=VectorParameter() {parameterName=Color, defaultValue=<rgba 1 0.9598 0.3008 1>}.RGB
    Fraction=ScalarParameter() {parameterName=DefaultColorDes}
   )
   B=Power(Base=TextureSample(UVs=@n8) {samplerType=SAMPLERTYPE_LinearGrayscale, texture=/Game/Effects/Fort_Effects/Emotes/UltraLight/Texture/T_PlasmaCoils}.R)
  )
  n10 = TextureCoordinate()
  n11 = VectorParameter() {parameterName=OpacityMaskSpeed, defaultValue=<rgba -1 0 0 0>}
  n12 = TextureSample(UVs=Panner(Coordinate=Add(A=@n10, B=@n6), Time=@n7, Speed=ComponentMask(None=@n11.RGB) {g=true, r=true}) {speedX=-1}) {samplerType=SAMPLERTYPE_Grayscale, texture=&T_Smoke_Trail_Soft}
  n13 = Multiply(A=VectorParameter() {parameterName=ColorDark, defaultValue=<rgba 0.401 0.401 0.401 1>}.RGB, B=@n12.R)
  MaterialFunctionCall(Emissive=Multiply(
    A=MaterialFunctionCall(Value_Morning=@n1, Value_Day=@n2, Value_Evening=@n3, Value_Night=@n4) {materialFunction=&Scalar_Time_of_Day}.Result
    B=Add(A=@n9, B=@n13)
   ) {constA=1}) {materialFunction=&MF_FX_EmissiveAdjustments}
  n14 = TextureCoordinate()
  n15 = VectorParameter() {parameterName=OpacityNoiseSpeed, defaultValue=<rgba -1.5 0 0 0>}
  n16 = TextureSample(UVs=Panner(Coordinate=Add(A=@n6, B=@n14), Time=@n7, Speed=ComponentMask(None=@n15.RGB) {g=true, r=true}) {speedX=-1.5}) {samplerType=SAMPLERTYPE_Grayscale, texture=&LowResBlurredNoise}
  n17 = OneMinus(None=ComponentMask(None=TextureCoordinate()) {r=true})
  n18 = Power(Base=@n17, Exp=ScalarParameter() {defaultValue=20, parameterName=Param_1})
  n19 = ScalarParameter() {parameterName=Param_2}
  n20 = Add(A=Add(A=@n12.R, B=Multiply(A=@n16.R, B=@n12.R)) {constB=0}, B=Saturate(None=Multiply(A=@n18, B=@n19)))
  n21 = Power(Base=@n17, Exp=ScalarParameter())
  n22 = ComponentMask(None=TextureCoordinate()) {r=true}
  n23 = Multiply(A=ParticleColor().A, B=Multiply(A=Multiply(A=@n20, B=@n21), B=Sine(None=@n22) {period=2}))
  n24 = ScalarParameter() {parameterName=CameraFadeDistance}
  n25 = DepthFade(
   Opacity=Multiply(A=@n23, B=MaterialFunctionCall(Distance=@n24) {materialFunction=&Near_Camera_Fade}.Result)
   FadeDistance=ScalarParameter() {defaultValue=10, parameterName=FadeDistance}
  ) {fadeDistanceDefault=10}
  n26 = ShadingPathSwitch(Default=@n25, Deferred=@n25, Forward=@n23, Mobile=@n23)
  n27 = NamedRerouteUsage()
  n28 = OneMinus(None=SphereMask(
    A=Saturate(None=@n27)
    B=Constant()
    Radius=ScalarParameter() {defaultValue=1, parameterName=FadeAngle Radius}
    Hardness=ScalarParameter() {defaultValue=0.3, parameterName=FadeAngle Hardness}
   ) {attenuationRadius=0.5, hardnessPercent=0})
  n29 = TextureSample(UVs=Add(A=TextureCoordinate(), B=Constant2Vector())) {samplerType=SAMPLERTYPE_Grayscale, texture=&LowResBlurredNoise}
  Multiply(
   A=Multiply(A=Multiply(A=@n26, B=@n28), B=ScalarParameter() {defaultValue=0.7, parameterName=AlphaIntensity}) {constB=0.7}
   B=Saturate(None=LinearInterpolate(Alpha=@n29.R) {constA=-0.1, constB=2})
  )
  n30 = Constant3Vector() {constant=<rgba 0 0 1 0>}
  n31 = Multiply(A=Time(), B=ScalarParameter() {defaultValue=0.2, parameterName=RotateSpeed}) {constB=0.2}
  n32 = MaterialFunctionCall(Rotation Axis=@n30, Rotation Angle=@n31) {materialFunction=&FixRotateAboutAxisNormals}
  n33 = CameraVectorWS()
  NamedRerouteDeclaration(None=Abs(None=DotProduct(A=@n32.New World Space Vertex Normal, B=@n33))) {name=Fixed Normals, nodeColor=<rgba 0.6791 0.0155 1 1>}
  n34 = ObjectPositionWS()
  n35 = WorldPosition()
  n36 = VectorParameter() {parameterName=WPODirection, defaultValue=<rgba 0 1 0 1>}
  n37 = TextureCoordinate()
  n38 = ScalarParameter() {parameterName=UVOffset}
  n39 = Add(A=ComponentMask(None=Add(A=@n37, B=@n38)) {r=true}, B=Multiply(A=Time()) {constB=-0.5})
  n40 = Multiply(
   A=Multiply(A=Transform(None=@n36.RGB) {transformType=TRANSFORM_View}, B=Sine(None=@n39) {period=0.5})
   B=ScalarParameter() {defaultValue=20, parameterName=WPOIntensity}
  ) {constB=10}
  n41 = ComponentMask(None=TextureCoordinate()) {r=true}
  Reroute(None=Add(A=Multiply(A=@n40, B=@n41), B=RotateAboutAxis(NormalizedRotationAxis=@n30, RotationAngle=@n31, PivotPoint=@n34, Position=@n35.XYZ)))
 ]
}
