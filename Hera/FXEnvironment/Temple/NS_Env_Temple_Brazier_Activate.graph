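# Niagara system for the activation effect of a temple brazier in the Hera environment FX set.
# It layers eight emitters, named for soft, triangle and eye glows, light beams, two radial energy bursts, sparks and squares.
# It exposes no user variables, and SystemState is the only listed module dependency, with no renderers, data interfaces or dynamic inputs recorded.
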NiagaraSystem /Hera/FXEnvironment/Temple/NS_Env_Temple_Brazier_Activate
dependencies {usedDataInterfaces [], usedDynamicInputs [], usedModules [&SystemState]}
emitters {
 Big_Soft_Glow {rendererClasses [], renderers []}
 EnergyBurst_Radial {rendererClasses [], renderers []}
 EnergyBurst_Radial001 {rendererClasses [], renderers []}
 Eyes_Glow {rendererClasses [], renderers []}
 Light_Beams {rendererClasses [], renderers []}
 Sparks {rendererClasses [], renderers []}
 Squares_LW {rendererClasses [], renderers []}
 Triangle_Glow {rendererClasses [], renderers []}
}
summary {
 systemName=NS_Env_Temple_Brazier_Activate
 emitters [
  {emitterName=Big_Soft_Glow, rendererClasses []}
  {emitterName=Triangle_Glow, rendererClasses []}
  {emitterName=Light_Beams, rendererClasses []}
  {emitterName=EnergyBurst_Radial, rendererClasses []}
  {emitterName=Eyes_Glow, rendererClasses []}
  {emitterName=EnergyBurst_Radial001, rendererClasses []}
  {emitterName=Sparks, rendererClasses []}
  {emitterName=Squares_LW, rendererClasses []}
 ]
 userVariables []
}
