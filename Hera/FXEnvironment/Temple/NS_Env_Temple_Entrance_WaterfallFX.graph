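# Niagara system for the waterfall effect at the temple entrance in the environment FX set.
# It layers seven emitters, named for splashes, fog, foamy water, bubbles, water ripples, dripping drops and overspray at the mouth.
# SystemState is the only listed module, and the system exposes no user variables.
# Emitter renderers and internal logic are not recorded here, so each layer's role rests on its name.
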NiagaraSystem /Hera/FXEnvironment/Temple/NS_Env_Temple_Entrance_WaterfallFX
dependencies {usedDataInterfaces [], usedDynamicInputs [], usedModules [&SystemState]}
emitters {
 Bubbles001 {rendererClasses [], renderers []}
 FoamyWater {rendererClasses [], renderers []}
 Fog001 {rendererClasses [], renderers []}
 Splashes {rendererClasses [], renderers []}
 WaterDrops_Dripping {rendererClasses [], renderers []}
 WaterDrops_Mouth_Overspray {rendererClasses [], renderers []}
 Water_Ripple {rendererClasses [], renderers []}
}
summary {
 systemName=NS_Env_Temple_Entrance_WaterfallFX
 emitters [
  {emitterName=Splashes, rendererClasses []}
  {emitterName=Fog001, rendererClasses []}
  {emitterName=FoamyWater, rendererClasses []}
  {emitterName=WaterDrops_Dripping, rendererClasses []}
  {emitterName=Bubbles001, rendererClasses []}
  {emitterName=Water_Ripple, rendererClasses []}
  {emitterName=WaterDrops_Mouth_Overspray, rendererClasses []}
 ]
 userVariables []
}
