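# Niagara system for the flames of a fire trap in the temple environment.
# It layers six emitters: Embers, MouthSparks, FlameJetMesh, FlameFlipbook, Fire001 and Light, whose names point to sparks, a mesh jet, flipbook flames and a light.
# The system exposes no user variables and lists SystemState as its only module dependency; emitter internals and renderers are not recorded here.
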NiagaraSystem /Hera/FXEnvironment/Temple/NS_Env_Temple_FireTrap_Flames
dependencies {usedDataInterfaces [], usedDynamicInputs [], usedModules [&SystemState]}
emitters {
 Embers {rendererClasses [], renderers []}
 Fire001 {rendererClasses [], renderers []}
 FlameFlipbook {rendererClasses [], renderers []}
 FlameJetMesh {rendererClasses [], renderers []}
 Light {rendererClasses [], renderers []}
 MouthSparks {rendererClasses [], renderers []}
}
summary {
 systemName=NS_Env_Temple_FireTrap_Flames
 emitters [
  {emitterName=Embers, rendererClasses []}
  {emitterName=MouthSparks, rendererClasses []}
  {emitterName=FlameJetMesh, rendererClasses []}
  {emitterName=FlameFlipbook, rendererClasses []}
  {emitterName=Fire001, rendererClasses []}
  {emitterName=Light, rendererClasses []}
 ]
 userVariables []
}
