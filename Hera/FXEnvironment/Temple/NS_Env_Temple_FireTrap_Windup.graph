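# Niagara system for the windup telegraph of a temple fire trap, apparently a glow on the trap's eye.
# It has two emitters, Eye_Glow_Inner and Eye_Glow_Outer, with no renderers listed here.
# A float user parameter, WindupGlow, is exposed, presumably to drive the glow intensity.
# The system scripts use System State and a Set Variables module, with curve-sampled float and Make Linear Color dynamic inputs also in use.
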NiagaraSystem /Hera/FXEnvironment/Temple/NS_Env_Temple_FireTrap_Windup
dependencies {
 usedDataInterfaces [&NiagaraDataInterfaceCurve]
 usedDynamicInputs [&MakeLinearColor, &FloatFromCurve]
 usedModules [
  &SystemState
  &self:SystemSpawnScript.SystemScriptSource.SystemScriptGraph.NiagaraNodeAssignment_0.SetVariables_#1
 ]
}
emitters {Eye_Glow_Inner {rendererClasses [], renderers []}, Eye_Glow_Outer {rendererClasses [], renderers []}}
summary {
 systemName=NS_Env_Temple_FireTrap_Windup
 emitters [{emitterName=Eye_Glow_Inner, rendererClasses []}, {emitterName=Eye_Glow_Outer, rendererClasses []}]
 userVariables [User.WindupGlow {type=NiagaraFloat, defaultValue=0, description=""}]
}
