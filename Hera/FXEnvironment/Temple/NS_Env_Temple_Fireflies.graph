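# Temple ambient Niagara system that crossfades butterflies around midday with fireflies toward night through time-of-day spawn-rate curves.
# Butterflies drift in two partitioned directions with oscillating vertical force, drawn as meshes with tinted sprites.
# GPU fireflies combine curl noise, gust-driven wind, drag, repulsion and distance-field avoidance, with waveform jitter flickering their sprite glow.
# Both spawn in wide, vertically flattened volumes and scale in and out over each particle's life.
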NiagaraSystem /Hera/FXEnvironment/Temple/NS_Env_Temple_Fireflies
dependencies {
 usedDataInterfaces [
  &NiagaraDataInterfaceCurve
  &NiagaraDataInterfaceCollisionQuery
  &NiagaraDataInterfaceColorCurve
  &NiagaraDataInterfaceCamera
  &NiagaraDataInterfaceArrayFloat3
 ]
 usedDynamicInputs []
 usedModules [&SystemState]
}
emitters {
 Butterfly {
  emitterUpdateScript [
   &EmitterState(Enable Distance Culling=true, MaxDistance=8000, Module.bUseMaxDistance=true, Module.bUseMinDistance=false, Scalability Mode=::NewEnumerator1)
   &SpawnRate(
    SpawnRate=&FloatFromCurve(
     FloatCurve=DataInterface() {propertyValues {Curve {
        defaultValue=1
        keys [
         {interpMode=RCIM_Cubic, leaveTangentWeight=0.1, tangentMode=RCTM_User, time=0, value=0}
         {arriveTangentWeight=0.1, interpMode=RCIM_Cubic, leaveTangentWeight=0.1333, tangentMode=RCTM_Break, time=0.3, value=1}
         {
          arriveTangent=-0.4792
          arriveTangentWeight=0.1478
          interpMode=RCIM_Cubic
          leaveTangent=-0.4792
          leaveTangentWeight=0.1109
          tangentMode=RCTM_User
          time=0.7
          value=1
         }
         {tangentMode=RCTM_User, time=1, value=0}
        ]
       }}}
     CurveIndex=Linked() {linkedVariable {name=NPC.FortniteNPC.TimeOfDay_0-1, type=NiagaraFloat}}
     Scale Curve=&RandomRangeFloat(Minimum=6, Maximum=8)
    )
    Use Spawn Probability=false
   ) {module=SpawnRate002}
   &SpawnRate(
    SpawnRate=&FNTimeOfDayScalar(Night Scale=0, Morning Scale=5, Day Scale=5, Evening Scale=0)
    Use Spawn Probability=false
   ) {enabled=false}
  ]
  particleSpawnScript [
   &PartitionParticles()
   &InitializeParticle(
    Color Mode=::NewEnumerator0
    Lifetime=&RandomRangeFloat(Minimum=3, Maximum=4)
    Mesh Scale Mode=::NewEnumerator3
    Mesh Uniform Scale=0.03
    UsePositionOffset=false
    Write Material Random=true
   )
   &InitialMeshOrientation(Mesh Axis To Orient=<0 0 1>, Mesh Orientation Mode=::NewEnumerator2, Use Rotation Vector=false)
   &ShapeLocation(
    Handle Radius=2000
    Large Radius=2500
    MaskSpawnGroup=false
    Non Uniform Scale=<1 1 0.3>
    Offset Mode=::NewEnumerator0
    Override Seed=false
    Shape Primitive=::NewEnumerator3
    UseOwnerScale=true
   ) {module=ShapeLocation001}
   &AddVelocity(Velocity=&SelectVectorFromArray(
     Direct Array Index=$Particles.PartitionParticles.Partition
     Vector Selection Array=DataInterface() {propertyValues {FloatData [<-50 -50 -20>, <50 50 10>]}}
    ))
  ]
  particleUpdateScript [
   &ParticleState()
   &DynamicMaterialParameters(
    Index 0 Param 1=0.2364
    Index 0 Param 4=&Multiply_Float(A=$System.NormalizedLoopAge, B=3)
    Index 1 Param 1=3000
    Index 1 Param 2=1
    Module.Param0WriteEnabled=true
    Module.Param0WriteEnabled1=true
    Module.Param1WriteEnabled=true
    Module.Param1WriteEnabled1=true
    Module.Param2WriteEnabled=true
    Module.Param2WriteEnabled1=true
    Module.Param3WriteEnabled=true
    Module.Param3WriteEnabled1=true
    Write Parameter Index 1=true
   )
   &LinearForce(Force=&Multiply_VectorByFloat(
     Vector=<0 0 20>
     Float=&FloatFromCurve(FloatCurve=DataInterface() {propertyValues {Curve {keys [
          <at 0 0 RCIM_Cubic>
          {arriveTangent=-1.8589, interpMode=RCIM_Cubic, leaveTangent=-1.8589, time=0.2151, value=1.0694}
          {arriveTangent=-0.0884, interpMode=RCIM_Cubic, leaveTangent=-0.0884, time=0.5948, value=-1.1057}
          <at 1 1 RCIM_Cubic>
         ]}}})
    ))
   &SolveForcesAndVelocity(Clamp Velocity=false, Limit Acceleration=false)
   &ScaleMeshSize(Scale Factor=&VectorFromFloat(Value=&FloatFromCurve(FloatCurve=DataInterface() {propertyValues {Curve {keys [
          <at 0 0 RCIM_Cubic>
          {arriveTangent=1.1765, interpMode=RCIM_Cubic, leaveTangent=1.1765, time=0.15, value=1}
          {arriveTangent=-1.1765, interpMode=RCIM_Cubic, leaveTangent=-1.1765, time=0.85, value=1}
          <at 1 0 RCIM_Cubic>
         ]}}})))
   &FN_Sprite_Attributes(
    Color=<rgba 0 0.5183 1 0.1>
    Create Color Attribute=true
    Create SpriteSize Attribute=true
    SpriteSize=&Multiply_Vector2D_ByFloat(Float=&MakeFloatFromVector(Vector=$Particles.Scale, Channel=::NewEnumerator0), Vector2D=<1200 1200>)
    Write to Module Namespace=true
   )
  ]
  rendererClasses [&NiagaraSpriteRendererProperties, &NiagaraMeshRendererProperties]
  renderers [
   {rendererClass=&NiagaraSpriteRendererProperties}
   {rendererClass=&NiagaraMeshRendererProperties, rendererIndex=1}
  ]
 }
 Firefiles_Epic {
  simTarget=GPUComputeSim
  emitterSpawnScript [
   /SetVariables(Emitter.Camera Position:NiagaraPosition=<-77365 12953 3500>) {enabled=false, isSetParameters=true}
   /SetVariables(Emitter.ScaleMult=1) {isSetParameters=true}
  ]
  emitterUpdateScript [
   &EmitterState()
   &SpawnRate(
    SpawnRate=&FloatFromCurve(
     FloatCurve=DataInterface() {propertyValues {Curve {
        defaultValue=1
        keys [
         {interpMode=RCIM_Cubic, tangentMode=RCTM_User, time=0, value=1}
         {interpMode=RCIM_Cubic, tangentMode=RCTM_Break, time=0.3, value=0}
         {arriveTangent=0.4792, interpMode=RCIM_Cubic, leaveTangent=0.4792, tangentMode=RCTM_User, time=0.7, value=0}
         {tangentMode=RCTM_User, time=1, value=1}
        ]
       }}}
     CurveIndex=Linked() {linkedVariable {name=NPC.FortniteNPC.TimeOfDay_0-1, type=NiagaraFloat}}
     Scale Curve=&RandomRangeFloat(Minimum=9, Maximum=12)
    )
    Use Spawn Probability=false
   ) {module=SpawnRate002}
  ]
  particleSpawnScript [
   &InitializeParticle(
    Color Maximum=<rgba 0.003 1 0.7611 1>
    Color Minimum=<rgba 0.0028 0.488 1 1>
    Color Mode=::NewEnumerator2
    Lifetime=5
    Mesh Scale Mode=::NewEnumerator3
    Mesh Uniform Scale=&Multiply_Float(A=&RandomRangeFloat(Minimum=0.36, Maximum=0.6), B=$Emitter.ScaleMult)
    Sprite Size Mode=::NewEnumerator3
    Uniform Sprite Size=&Multiply_Float(A=75, B=$Emitter.ScaleMult)
    UsePositionOffset=false
    Write Material Random=true
   ) {module=InitializeParticle001}
   &ShapeLocation(
    Handle Radius=2000
    Large Radius=2500
    MaskSpawnGroup=false
    Non Uniform Scale=<1 1 0.3>
    Offset Mode=::NewEnumerator0
    Override Seed=false
    Shape Primitive=::NewEnumerator3
    UseOwnerScale=true
   ) {module=ShapeLocation001}
   &CameraOffset(Camera Offset Amount=1)
   &AccelerationForce(Acceleration=&Multiply_VectorByFloat(Vector=$Output.SampleLandscape_02.LandscapeNormal, Float=300)) {module=AccelerationForce001}
   &ApplyInitialForces()
  ]
  particleUpdateScript [
   &ParticleState()
   /SetVariables(Particles.Jitter:NiagaraFloat=&Waveform(
     "Waveform [1]"=::NewEnumerator12
     "[1] Frequency"=&RandomRangeFloat(Minimum=15, Maximum=22)
     "[1] Amplitude Min/Max"=&RandomRangeVector2D(Minimum=<0 1>, Maximum=<0.3 1>)
     "[1] Scale"=3
    )) {isSetParameters=true}
   &ScaleMeshSize(Scale Factor=&VectorFromFloat(Value=1)) {module=ScaleMeshSize001}
   &ScaleMeshSize(Scale Factor=&VectorFromFloat(Value=&FloatFromCurve(FloatCurve=DataInterface() {propertyValues {Curve {keys [
          <at 0 0 RCIM_Cubic>
          {arriveTangent=1.1765, leaveTangent=1.1765, time=0.1, value=1}
          {interpMode=RCIM_Cubic, tangentMode=RCTM_Break, time=0.8, value=1}
          {arriveTangent=-23.4104, interpMode=RCIM_Cubic, tangentMode=RCTM_Break, time=1, value=0}
         ]}}})))
   &ScaleSpriteSize(Uniform Curve Sprite Scale=DataInterface() {propertyValues {Curve {keys [
        <at 0 0 RCIM_Cubic>
        {arriveTangent=1.1765, leaveTangent=1.1765, time=0.1, value=1}
        {interpMode=RCIM_Cubic, tangentMode=RCTM_Break, time=0.8, value=1}
        {arriveTangent=-23.4104, interpMode=RCIM_Cubic, tangentMode=RCTM_Break, time=1, value=0}
       ]}}})
   &CurlNoiseForce(Bias Noise=false, Noise Frequency=5, Noise Strength=750, Pan Noise=false)
   &WindForce(Wind Speed Scale=&Lerp_Float(A=0.1, B=5, Alpha=$NPC.FortniteNPC.WindGustsSystem_Active))
   &Drag(Drag=10, Use Linear Drag=true, Use Rotational Drag=true)
   &AccelerationForce(Acceleration=<50 0 300>)
   &PointAttractionForce(Attraction Radius=150, Attraction Strength=-500, Kill Within Radius=false, Module.Use Falloff=true, UseAttractorOffset=false)
   /Niagara/Modules/Math/AvoidDistanceFieldSurfaces_GPU(Avoid Nearest Surface=true, Avoid Oncoming Surfaces=true, Nearest Surface Avoidance Strength=5)
   &SolveForcesAndVelocity(Clamp Velocity=false, Limit Acceleration=false)
   &DynamicMaterialParameters(
    Index 0 Param 4=&Lerp_Float(A=10, B=25, Alpha=&Sine(Normalized Angle=$Particles.Age, Period=1, Scale=1, Bias=0.5))
    Index 1 Param 1=1
    Module.Index0Param0Write=true
    Module.Index0Param1Write=true
    Module.Index0Param2Write=true
    Module.Index0Param3Write=true
    Module.Index1Param0Write=true
    Module.Index1Param1Write=true
    Module.Index1Param2Write=true
    Module.Index1Param3Write=true
    Write Parameter Index 1=true
   )
   &ScaleColor(
    Curve Index=&Multiply_Float(
     A=$Particles.NormalizedAge
     B=&FloatFromCurve(
      FloatCurve=DataInterface() {propertyValues {Curve {
         defaultValue=0.1
         keys [
          {interpMode=RCIM_Cubic, tangentMode=RCTM_User, time=0, value=1}
          {arriveTangent=0.1559, interpMode=RCIM_Cubic, leaveTangent=0.1559, tangentMode=RCTM_User, time=0.2, value=0}
          {interpMode=RCIM_Cubic, tangentMode=RCTM_User, time=0.8, value=0}
          {tangentMode=RCTM_User, time=1, value=1}
         ]
        }}}
      CurveIndex=Linked() {linkedVariable {name=NPC.FortniteNPC.TimeOfDay_0-1, type=NiagaraFloat}}
     )
    )
    Linear Color Curve=DataInterface() {propertyValues {AlphaCurve {keys [<at 0.8906 0.45>,<at 1 0>]}, BlueCurve {keys [<at 0 0.5>]}, GreenCurve {keys [<at 0 0.5>]}, RedCurve {keys [<at 0 0.5>]}}}
    Scale Mode=::NewEnumerator2
   )
   &InterpolateOverTime(Rate Of Change=5, Target Position Value:NiagaraPosition=$Particles.Position, Variable Type=::NewEnumerator6)
   /SetVariables(Particles.Vector 4:Vector4f=&MakeVector4(
     X=&MakeFloatFromVector(
      Vector=&Subtract_Position(A=$Particles.InterpolateOverTime.InterpolateOverTime_Position.Moving Average, B=$Particles.Position)
      Channel=::NewEnumerator0
     )
     Y=&MakeFloatFromVector(
      Vector=&Subtract_Position(A=$Particles.InterpolateOverTime.InterpolateOverTime_Position.Moving Average, B=$Particles.Position)
      Channel=::NewEnumerator1
     )
     Z=&MakeFloatFromVector(
      Vector=&Subtract_Position(A=$Particles.InterpolateOverTime.InterpolateOverTime_Position.Moving Average, B=$Particles.Position)
      Channel=::NewEnumerator2
     )
     W=1
    )) {isSetParameters=true}
   &ScaleMeshSize(Scale Factor=&VectorFromFloat(Value=&Lerp_Float(
      A=0
      B=1
      Alpha=&RemapRange(
       Input Value To Remap=&VectorLength(Vector=&SubtractVectorFromPosition(
         Position=&ReturnCameraPosition(Coordinate Space:ENiagaraCoordinateSpace=::Simulation, Camera Query:NiagaraDataInterfaceCamera=DataInterface())
         Vector=<0 0 0>
        ))
       Input Min=0
       Input Max=1
       Output Min=0
       Output Max=1
      )
     ))) {module=ScaleMeshSize002}
   &FN_Sprite_Attributes(
    Camera Offset=20
    Color=&Multiply_LinearColorByFloat(
     Linear Color=&HueShiftLinearColor(Hue=0, Saturation=0.25, Value=&Multiply_Float(A=5, B=$Particles.Jitter), Linear Color=$Particles.Color)
     Float=&FloatFromCurve(FloatCurve=DataInterface() {propertyValues {Curve {keys [
          <at 0.05 0 RCIM_Cubic>
          {arriveTangent=1.1765, time=0.2, value=1}
          {interpMode=RCIM_Cubic, tangentMode=RCTM_Break, time=0.8, value=1}
          {arriveTangent=-13.7968, interpMode=RCIM_Cubic, tangentMode=RCTM_Break, time=1, value=0}
         ]}}})
    )
    Create Camera Offset Attribute=true
    Create Color Attribute=true
    Create Position Attribute=true
    Create SpriteSize Attribute=true
    Position:NiagaraPosition=&AddVectorToPosition(
     Position=$Particles.Position
     Vector=&TransformVector(Vector=<0 0 0>, Source Space=::World, Destination Space=::Local, Apply Scale=true)
    )
    SpriteSize=&Multiply_Vector2D_ByFloat(
     Float=&Lerp_Float(
      A=0
      B=1
      Alpha=&RemapRange(
       Input Value To Remap=&VectorLength(Vector=&SubtractVectorFromPosition(
         Position=/Niagara/DynamicInputs/Helpers/SimulationPosition()
         Vector=Linked() {linkedVariable {name=Particles.Position, type=NiagaraPosition}}
        ))
       Input Min=0
       Input Max=2500
       Output Min=0
       Output Max=1
      )
     )
     Vector2D=&Multiply_Vector2D_ByFloat(Float=1, Vector2D=<40 40>)
    )
    Write to Module Namespace=true
   )
   /SetVariables(Particles.Vector 4:Vector4f=<1 1 0 1>) {isSetParameters=true}
  ]
  rendererClasses [&NiagaraMeshRendererProperties, &NiagaraSpriteRendererProperties]
  renderers [
   {rendererClass=&NiagaraMeshRendererProperties, rendererIndex=0}
   {rendererClass=&NiagaraSpriteRendererProperties, rendererIndex=1}
   {rendererClass=&NiagaraSpriteRendererProperties, rendererIndex=2}
  ]
 }
 Thumbnail {bEnabled=false, rendererClasses [], renderers []}
}
summary {
 systemName=NS_Env_Temple_Fireflies
 emitters [
  {
   emitterName=Firefiles_Epic
   simTarget=GPUComputeSim
   rendererClasses [&NiagaraMeshRendererProperties, &NiagaraSpriteRendererProperties]
  }
  {bEnabled=false, emitterName=Thumbnail, rendererClasses []}
  {emitterName=Butterfly, rendererClasses [&NiagaraSpriteRendererProperties, &NiagaraMeshRendererProperties]}
 ]
 userVariables [
  User.Landscape {
   type=NiagaraDataInterfaceLandscape
   description=""
   defaultValue {
    struct=NiagaraExt_VariableValue_DataInterface
    value {
     dataInterface=&self:NiagaraDataInterfaceLandscape_0
     dataInterfaceClass=&NiagaraDataInterfaceLandscape
    }
   }
  }
  User.ShapeScale {type=NiagaraFloat, defaultValue=1, description=""}
 ]
}
