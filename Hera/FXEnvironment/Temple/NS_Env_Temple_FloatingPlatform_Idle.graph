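# Niagara system for the idle effect of a temple floating platform, built around one emitter named Sparkles_Up.
# At system level it uses SystemState and a Set Variables assignment in the system spawn script.
# It also uses a Float From Curve dynamic input backed by a curve data interface.
# The emitter lists no renderers and the system lists no user parameters.
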NiagaraSystem /Hera/FXEnvironment/Temple/NS_Env_Temple_FloatingPlatform_Idle
dependencies {
 usedDataInterfaces [&NiagaraDataInterfaceCurve]
 usedDynamicInputs [&FloatFromCurve]
 usedModules [
  &SystemState
  &self:SystemSpawnScript.SystemScriptSource.SystemScriptGraph.NiagaraNodeAssignment_2.SetVariables_#1
 ]
}
emitters {Sparkles_Up {rendererClasses [], renderers []}}
summary {systemName=NS_Env_Temple_FloatingPlatform_Idle, emitters [{emitterName=Sparkles_Up, rendererClasses []}], userVariables []}
