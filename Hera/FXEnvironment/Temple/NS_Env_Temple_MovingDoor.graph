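# Niagara system for a moving door in the Temple environment, named for debris and dust effects.
# It holds three emitters: Ground_Debris, Roof_Debris and Dust, with no renderers listed for any of them.
# SystemState is the only listed module, and the system exposes no user variables, data interfaces or dynamic inputs.
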NiagaraSystem /Hera/FXEnvironment/Temple/NS_Env_Temple_MovingDoor
dependencies {usedDataInterfaces [], usedDynamicInputs [], usedModules [&SystemState]}
emitters {Dust {rendererClasses [], renderers []}, Ground_Debris {rendererClasses [], renderers []}, Roof_Debris {rendererClasses [], renderers []}}
summary {
 systemName=NS_Env_Temple_MovingDoor
 emitters [{emitterName=Ground_Debris, rendererClasses []}, {emitterName=Roof_Debris, rendererClasses []}, {emitterName=Dust, rendererClasses []}]
 userVariables []
}
