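# Niagara system for the water effect of a temple shrine fountain, with a KillFountain bool exposed as a user parameter.
# Separate emitters cover splashes, fog, foamy water, bubbles, dripping drops and mouth overspray, while a low foam variant is disabled.
# System-level modules include a Timeline and four variable assignments.
# Dynamic inputs include curve, sine, lerp and time-of-day color, though their wiring is not shown.
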NiagaraSystem /Hera/FXEnvironment/Temple/NS_Env_Temple_Shrine_FountainFX
dependencies {
 usedDataInterfaces [&NiagaraDataInterfaceArrayFloat, &NiagaraDataInterfaceCurve]
 usedDynamicInputs [
  &Multiply_VectorByFloat
  &VectorFromFloat
  &Sine
  &Multiply_Float
  &FloatFromCurve
  &FNTimeOfDayColor
  &Lerp_Vector
 ]
 usedModules [
  &SystemState
  &Timeline
  &self:SystemSpawnScript.SystemScriptSource.SystemScriptGraph.NiagaraNodeAssignment_1.SetVariables_#1
  &self:SystemSpawnScript.SystemScriptSource.SystemScriptGraph.NiagaraNodeAssignment_0.SetVariables_#2
  &self:SystemSpawnScript.SystemScriptSource.SystemScriptGraph.NiagaraNodeAssignment_2.SetVariables_#3
  &self:SystemSpawnScript.SystemScriptSource.SystemScriptGraph.NiagaraNodeAssignment_3.SetVariables_#4
 ]
}
emitters {
 Bubbles001 {rendererClasses [], renderers []}
 FoamyWater {rendererClasses [], renderers []}
 FoamyWater_Low {bEnabled=false, rendererClasses [], renderers []}
 Fog001 {rendererClasses [], renderers []}
 Splashes {rendererClasses [], renderers []}
 WaterDrops_Dripping {rendererClasses [], renderers []}
 WaterDrops_Mouth_Overspray {rendererClasses [], renderers []}
}
summary {
 systemName=NS_Env_Temple_Shrine_FountainFX
 emitters [
  {emitterName=Splashes, rendererClasses []}
  {emitterName=Fog001, rendererClasses []}
  {emitterName=FoamyWater, rendererClasses []}
  {emitterName=WaterDrops_Dripping, rendererClasses []}
  {bEnabled=false, emitterName=FoamyWater_Low, rendererClasses []}
  {emitterName=Bubbles001, rendererClasses []}
  {emitterName=WaterDrops_Mouth_Overspray, rendererClasses []}
 ]
 userVariables [User.KillFountain {type=NiagaraBool}]
}
